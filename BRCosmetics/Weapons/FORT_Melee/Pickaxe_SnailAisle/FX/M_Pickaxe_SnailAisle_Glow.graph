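# Unlit, two-sided translucent glow material for Niagara mesh particles in the SnailAisle pickaxe effects.
# Particle color RGB feeds the shared FX emissive-adjustment function.
# Particle alpha times vertex red is raised to a per-particle GradiantPower dynamic parameter.
# That gradient is multiplied by a procedural U-axis falloff, built from U times one-minus-U to fade both side edges, and by depth fade to soften intersections.
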Material /BRCosmetics/Weapons/FORT_Melee/Pickaxe_SnailAisle/FX/M_Pickaxe_SnailAisle_Glow
graph {
 properties {
  bAutomaticallySetUsageInEditor=false
  bUsedWithNiagaraMeshParticles=true
  blendMode=BLEND_Translucent
  shadingModel=MSM_Unlit
  twoSided=true
  usedShadingModels=Unlit
 }
 nodes [
  n1 = ParticleColor()
  MaterialFunctionCall(Emissive=@n1.RGB) {materialFunction=&MF_FX_EmissiveAdjustments}
  n2 = Power(Base=Multiply(A=@n1.A, B=VertexColor().R), Exp=DynamicParameter() {paramNames [GradiantPower, Param2, Param3, Param4]}.GradiantPower)
  n3 = ComponentMask(None=TextureCoordinate()) {r=true}
  n4 = Multiply(A=Multiply(A=@n3, B=OneMinus(None=@n3)), B=ScalarParameter() {defaultValue=1, parameterName=Param_1}) {constB=4}
  Multiply(A=Multiply(A=@n2, B=Saturate(None=@n4)), B=DepthFade(FadeDistance=ScalarParameter() {defaultValue=100, parameterName=DepthFade}))
  Multiply(A=@n2, B=ScalarParameter() {defaultValue=1})
  ScalarParameter() {defaultValue=2, parameterName=Power}
 ]
}
