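# Physics asset for the SnailAisle two-handed pickaxe mesh, a Battle Royale cosmetic melee weapon.
# A single capsule body on the Melee_TwoHanded_Handle bone serves as the weapon's simplified collision shape, offset along the bone's Z axis.
# It has no constraints, so the asset defines one rigid volume with no jointed or separately simulated parts.
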PhysicsAsset /BRCosmetics/Weapons/FORT_Melee/Pickaxe_SnailAisle/Meshes/SnailAisle_Axe_PhysicsAsset
bodies {Melee_TwoHanded_Handle [{shapeType=Capsule, length=85.971, radius=29.0314, center=<0.7904 -0.9863 41.0235>, rotation=<rot -0.0166 -90 -0.0012>}]}
constraints []
